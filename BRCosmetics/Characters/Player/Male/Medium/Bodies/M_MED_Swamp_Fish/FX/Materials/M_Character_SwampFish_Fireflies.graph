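# Translucent unlit Niagara mesh-particle material for fireflies on the Swamp Fish character cosmetic.
# Flapping rotates vertices around the particle position, using UV1 and UV2 masks for separate sine and cosine motions, plus a vertical bob.
# Particle-random time offsets and dynamic parameters blend low and high flap speeds.
# A desaturatable flies texture tinted by particle color drives emissive, while particle alpha times a scene-depth difference softens intersections.
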Material /BRCosmetics/Characters/Player/Male/Medium/Bodies/M_MED_Swamp_Fish/FX/Materials/M_Character_SwampFish_Fireflies
graph {
 properties {
  bUsedWithNiagaraMeshParticles=true
  blendMode=BLEND_Translucent
  pixelDepthOffsetMode=PDOM_Legacy
  shadingModel=MSM_Unlit
  usedShadingModels=Unlit
 }
 nodes [
  n1 = Constant()
  n2 = If(
   A=ComponentMask(None=TextureCoordinate() {coordinateIndex=1}) {r=true}
   B=Constant() {r=0.01}
   "A > B"=Constant() {r=1}
   "A == B"=@n1
   "A < B"=@n1
  )
  n3 = VectorParameter() {parameterName=RotationAxis, defaultValue=<rgba 1 0 0 1>}
  n4 = Add(A=Time(), B=ParticleRandom()) {constB=0}
  n5 = DynamicParameter() {paramNames [Flapping Speed Low, Flapping Speed High, Flapping Speed Blend, ""]}
  n6 = Multiply(A=@n4, B=@n5.Flapping Speed Low)
  n7 = Multiply(A=@n4, B=@n5.Flapping Speed High)
  n8 = LinearInterpolate(A=Sine(None=@n6), B=Sine(None=@n7), Alpha=@n5.Flapping Speed Blend)
  n9 = ScalarParameter() {defaultValue=0.25, parameterName=Flapping Range}
  n10 = RotateAboutAxis(
   NormalizedRotationAxis=Normalize(VectorInput=Transform(None=@n3.RGB) {transformSourceType=TRANSFORMSOURCE_Instance})
   RotationAngle=Multiply(A=@n8, B=@n9) {constB=0.25}
   PivotPoint=ParticlePositionWS()
   Position=WorldPosition().XYZ
  )
  n11 = VectorParameter() {parameterName=RotationAxis, defaultValue=<rgba 1 0 0 1>}
  n12 = Cosine(None=@n6)
  n13 = Cosine(None=@n7)
  n14 = RotateAboutAxis(
   NormalizedRotationAxis=Normalize(VectorInput=Transform(None=@n11.RGB) {transformSourceType=TRANSFORMSOURCE_Instance})
   RotationAngle=Multiply(A=LinearInterpolate(A=@n12, B=@n13, Alpha=@n5.Flapping Speed Blend), B=@n9) {constB=0.25}
   PivotPoint=ParticlePositionWS()
   Position=WorldPosition().XYZ
  )
  n15 = Constant()
  n16 = If(
   A=ComponentMask(None=TextureCoordinate() {coordinateIndex=2}) {r=true}
   B=Constant() {r=0.02}
   "A > B"=Constant() {r=1}
   "A == B"=@n15
   "A < B"=@n15
  )
  n17 = Multiply(A=ScalarParameter() {defaultValue=0.5, parameterName=Vertical Motion Range}, B=Constant3Vector() {constant=<rgba 0 0 1 0>})
  Add(
   A=Add(A=Multiply(A=@n2, B=@n10), B=Multiply(A=@n14, B=@n16))
   B=Multiply(A=@n8, B=Transform(None=@n17) {transformSourceType=TRANSFORMSOURCE_Instance})
  ) {constB=0}
  n18 = Desaturation(
   None=TextureSample() {texture=/BRCosmetics/Athena/Effects/Particles/Skydiving/NS_Contrail_StinkyCloud/T_Contrail_StinkyCloud_Flies}.RGB
   Fraction=ScalarParameter() {parameterName=Emissive_Saturation}
  )
  n19 = ParticleColor()
  MaterialFunctionCall(Emissive=Multiply(A=Multiply(A=@n18, B=@n19.RGB), B=ScalarParameter() {defaultValue=1, parameterName=Emissive_Intensity})) {materialFunction=&MF_FX_EmissiveAdjustments}
  n20 = SceneDepth()
  n21 = PixelDepth()
  Multiply(A=@n19.A, B=Saturate(None=Subtract(A=@n20.R, B=@n21.R)))
 ]
}
