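# Material function exposed to the library under Distance Fields that computes a signed distance field for a capped cylinder along the Y axis.
# It offsets Coordinates (UVs padded to 3D by default) by Center, then pairs radial XZ distance with absolute height.
# Subtracting scalar h sets both radius and half-height.
# The output adds the clamped interior distance to the exterior distance, giving negative values inside.
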MaterialFunction /Engine/Functions/DistanceFields/Shapes/DistanceField_Cylinder
graph {
 properties {bExposeToLibrary=true, description=Generates a distance field shape for a capped cylinder., libraryCategoriesText [Distance Fields]}
 nodes [
  n1 = TextureCoordinate() {desc=UVs}
  n2 = Constant()
  n3 = Subtract(
   A=FunctionInput(Preview=AppendVector(A=@n1, B=@n2)) {bUsePreviewValueAsDefault=true, inputName=Coordinates}
   B=FunctionInput() {bUsePreviewValueAsDefault=true, inputName=Center, sortPriority=1}
  )
  n4 = Distance(A=ComponentMask(None=@n3) {b=true, r=true}, B=Constant())
  n5 = ComponentMask(None=@n3) {g=true}
  n6 = FunctionInput() {bUsePreviewValueAsDefault=true, inputName=h, inputType=FunctionInput_Scalar, sortPriority=2, previewValue=<0.005 0 0 1>}
  n7 = Subtract(A=Abs(None=AppendVector(A=@n4, B=@n5)), B=@n6)
  n8 = Max(A=ComponentMask(None=@n7) {r=true}, B=ComponentMask(None=@n7) {g=true})
  n9 = Max(A=@n7) {constB=0}
  n10 = Constant()
  FunctionOutput(None=Add(A=Min(A=@n8) {constB=0}, B=Distance(A=@n9, B=@n10)))
 ]
}
