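# Additive unlit Niagara sprite material for glowing fireflies on the Swamp Fish character cosmetic.
# Particle color times a glow intensity feeds an emissive-adjustment function, while a radial falloff times particle alpha forms a soft dot.
# Depth fade softens that dot, bypassed on mobile and low quality.
# Time, a per-particle Randomness dynamic parameter and Frequency drive sine and cosine waves at different rates, building a wandering three-axis vector.
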Material /BRCosmetics/Characters/Player/Male/Medium/Bodies/M_MED_Swamp_Fish/FX/Materials/M_SwampFish_Fireflies
graph {
 properties {
  bUsedWithNiagaraSprites=true
  blendMode=BLEND_Additive
  ditherOpacityMask=true
  pixelDepthOffsetMode=PDOM_Legacy
  shadingModel=MSM_Unlit
  usedShadingModels=Unlit
 }
 nodes [
  n1 = Multiply(
   A=Add(A=DynamicParameter() {paramNames [Randomness,Param2,Param3,Param4]}.Randomness, B=Time())
   B=ScalarParameter() {defaultValue=1, parameterName=Frequency}
  )
  n2 = Multiply(A=@n1) {constB=0.8}
  n3 = Multiply(A=@n1) {constB=0.5}
  n4 = Multiply(A=@n1) {constB=1.45}
  MaterialFunctionCall(R=Multiply(A=Sine(None=@n2)) {constB=2}, G=Multiply(A=Cosine(None=@n3)) {constB=2}, B=Multiply(A=Cosine(None=@n4)) {constB=3}) {materialFunction=&AppendMany}
  n5 = ParticleColor()
  MaterialFunctionCall(Emissive=Multiply(A=ScalarParameter() {defaultValue=1, parameterName=GlowIntensityMasked, sortPriority=35}, B=@n5.RGB)) {materialFunction=&MF_FX_EmissiveAdjustments}
  n6 = TextureCoordinate()
  n7 = Constant() {r=0.5}
  n8 = Saturate(None=Divide(A=Distance(A=@n6, B=@n7), B=@n7))
  n9 = Multiply(A=@n5.A, B=Power(Base=OneMinus(None=@n8), Exp=ScalarParameter() {defaultValue=3, parameterName=Falloff Exponent}))
  Multiply(A=ScalarParameter() {defaultValue=1, parameterName=Mask_Intensity}, B=@n9)
  n10 = TextureCoordinate()
  n11 = Constant() {r=0.5}
  n12 = Subtract(A=ComponentMask(None=SceneTexture() {sceneTextureId=PPI_SceneDepth}.Color) {r=true}, B=Add(A=PixelDepth().R) {constB=-4.5})
  Multiply(A=Multiply(B=SphereMask(A=@n10, B=@n11) {attenuationRadius=0.3, hardnessPercent=1}), B=Saturate(None=Divide(A=@n12) {constB=0.4}))
  n13 = DepthFade(Opacity=@n9, FadeDistance=ScalarParameter() {defaultValue=2, parameterName=Depth Fade Distance})
  n14 = ShadingPathSwitch(Default=@n13, Deferred=@n13, Mobile=@n9) {desc=Migrated from FeatureLevelSwitch}
  QualitySwitch(Default=@n14, Low=@n9, High=@n14, Medium=@n14, Epic=@n14)
  Saturate(None=MaterialFunctionCall!(
    Coordinates=MaterialFunctionCall() {materialFunction=&LocalPosition}.Local Position
    Center=Constant3Vector()
    h=Constant() {r=20}
   ) {materialFunction=/Engine/Functions/DistanceFields/Shapes/DistanceField_Cylinder}.Result)
  StaticSwitchParameter(True=MaterialFunctionCall(
    Value_Morning=ScalarParameter() {defaultValue=1, parameterName=ToD_Morn_GlowIntensityMasked}
    Value_Day=ScalarParameter() {defaultValue=1, parameterName=ToD_Day_GlowIntensityMasked, sortPriority=33}
    Value_Evening=ScalarParameter() {defaultValue=1, parameterName=ToD_Eve_GlowIntensityMasked, sortPriority=34}
    Value_Night=ScalarParameter() {defaultValue=1, parameterName=ToD_Night_GlowIntensityMasked, sortPriority=35}
   ) {materialFunction=&Scalar_Time_of_Day}.Result) {parameterName=IsMaskedMaterial}
  Transform() {transformSourceType=TRANSFORMSOURCE_Local}
 ]
}
