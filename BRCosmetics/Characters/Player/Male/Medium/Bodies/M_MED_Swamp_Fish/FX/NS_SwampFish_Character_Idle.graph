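# Niagara idle effect for the Swamp Fish character: glowing fireflies swarming inside a jar, with a flickering teal light.
# Fireflies burst-spawn as mesh and sprite particles, stirred by curl noise and confined by pendulum constraints sized from user jar offset, radius and height.
# A system timeline ramps their agitation, brightness and camera offsets, with time-of-day scaling the base glow.
# A separate Light emitter multiplies waveform flicker by timeline time.
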NiagaraSystem /BRCosmetics/Characters/Player/Male/Medium/Bodies/M_MED_Swamp_Fish/FX/NS_SwampFish_Character_Idle
dependencies {
 usedDataInterfaces [&NiagaraDataInterfaceCurve]
 usedDynamicInputs [&SetBoolByFloatComparison]
 usedModules [
  &SystemState
  &self:SystemSpawnScript.SystemScriptSource.SystemScriptGraph.NiagaraNodeAssignment_2.SetVariables_#1
  &Timeline
  &self:SystemSpawnScript.SystemScriptSource.SystemScriptGraph.NiagaraNodeAssignment_0.SetVariables_#2
 ]
}
emitters {
 Fireflies {
  emitterUpdateScript [
   &EmitterState(
    Inactive Response=::NewEnumerator1
    Life Cycle Mode=::NewEnumerator1
    Loop Behavior=::NewEnumerator1
    Loop Duration Mode=::NewEnumerator1
    UseLoopDelay=false
   )
   &SpawnBurst_Instantaneous(Spawn Count=5, Use Spawn Probability=false, UseLoopCountLimit=false)
  ]
  particleSpawnScript [
   &InitializeParticle(
    Color Mode=::NewEnumerator0
    Mesh Scale Mode=::NewEnumerator1
    Mesh Uniform Scale Max=0.466
    Mesh Uniform Scale Min=0.333
    Position Offset:NiagaraPosition=$User.JarOffset
    Sprite Size Mode=::NewEnumerator3
    Uniform Sprite Size=11.5
    UsePositionOffset=true
    Write Material Random=true
   )
   /SetVariables(Particles.PositionCapsuleOffset:NiagaraPosition=&AddVectorToPosition(
     Position=$User.JarOffset
     Vector=&MakeVector(
      X=0
      Y=&Multiply_Float(A=&RandomRangeFloat(Minimum=-1, Maximum=1), B=$User.JarHeight)
      Z=0
     )
    )) {isSetParameters=true}
  ]
  particleUpdateScript [
   /SetVariables(Particles.ReactivityFadeInValue:NiagaraFloat=&FloatFromCurve(
     FloatCurve=DataInterface() {propertyValues {Curve {keys [<at 0 0>,<at 1 1>]}}}
     CurveIndex=&Multiply_Float(A=$System.Timeline.NormalizedTime, B=&RandomRangeFloat(Minimum=1, Maximum=7))
    )) {isSetParameters=true}
   /SetVariables(Particles.ReactivityFadeInValueColors:NiagaraFloat=&FNTimeOfDayScalar(
     Night Scale=&Lerp_Float(A=0.4, B=1, Alpha=$Particles.ReactivityFadeInValue)
     Morning Scale=$Particles.ReactivityFadeInValue
     Day Scale=$Particles.ReactivityFadeInValue
     Evening Scale=$Particles.ReactivityFadeInValue
    )) {enabled=false, isSetParameters=true}
   /SetVariables(Particles.ReactivityFadeInValueColors:NiagaraFloat=&Lerp_Float(
     A=&FNTimeOfDayScalar(Night Scale=0.5, Morning Scale=0.45, Day Scale=0.5, Evening Scale=0.47)
     B=1
     Alpha=$Particles.ReactivityFadeInValue
    )) {isSetParameters=true}
   /SetVariables(
    Particles.FireflyMeshCameraOffset:NiagaraFloat=&Lerp_Float(A=0, B=-3, Alpha=$Particles.ReactivityFadeInValueColors)
    Particles.FireflyMeshColor:LinearColor=&MakeLinearColorFromVectorAndFloat(
     "Vector (RGB)"=&Lerp_Vector(A=<0.6 0.6 0.6>, B=<4 4 4>, Alpha=$Particles.ReactivityFadeInValueColors)
     "Float (Alpha)"=&Lerp_Float(A=0.4, B=1, Alpha=$Particles.ReactivityFadeInValueColors)
    )
    Particles.GlowSpriteCameraOffset:NiagaraFloat=&Lerp_Float(A=-3, B=1, Alpha=$Particles.ReactivityFadeInValueColors)
    Particles.GlowSpriteColor:LinearColor=&MultiplyLinearColor(
     Linear Color=<rgba 0 1 0.7915 1>
     Scale Factor=&MakeVector4_FromFloat(Float=&Lerp_Float(
       A=0.133
       B=&Multiply_Float(
        A=0.5
        B=&Waveform(
         "Waveform [1]"=::NewEnumerator0
         "[1] Frequency"=&RandomRangeFloat(Minimum=0.4, Maximum=0.8)
         "[1] Amplitude Min/Max"=<0.6 1>
        )
       )
       Alpha=$Particles.ReactivityFadeInValueColors
      ))
    )
   ) {isSetParameters=true}
   &DynamicMaterialParameters(
    Index 0 Param 1=3
    Index 0 Param 2=9
    Index 0 Param 3=$System.Timeline.Time
    Module.Index0Param0Write=true
    Module.Index0Param1Write=true
    Module.Index0Param2Write=true
    Module.Index0Param3Write=false
   )
   &ParticleState(Kill Particles When Lifetime Has Elapsed=false, Let Infinitely Lived Particles Die When Emitter Deactivates=true)
   &CurlNoiseForce(
    Noise Frequency=&RandomRangeFloat(Minimum=111, Maximum=222)
    Noise Strength=&Lerp_Float(A=222, B=555, Alpha=$Particles.ReactivityFadeInValue)
    Pan Noise=true
    Pan Noise Field=&MakeVector(X=0, Y=0, Z=&Lerp_Float(A=1, B=6, Alpha=$Particles.ReactivityFadeInValue))
    Scale Curl Lookup By Engine Owner Scale=false
   )
   &CurlNoiseForce(
    Noise Frequency=&RandomRangeFloat(Minimum=50, Maximum=100)
    Noise Strength=&Lerp_Float(A=0, B=&RandomRangeFloat(Minimum=800, Maximum=1500), Alpha=$Particles.ReactivityFadeInValue)
    Pan Noise=true
    Pan Noise Field=&MakeVector(X=3, Y=4, Z=5)
    Scale Curl Lookup By Engine Owner Scale=false
   ) {module=CurlNoiseForce001}
   &PendulumSetup(
    Module.Enable Spring Driver=false
    Pendulum Length:NiagaraFloat=$User.JarRadius
    Pendulum Pivot Offset:NiagaraPosition=$Particles.PositionCapsuleOffset
    Pendulum Pivot Offset Coordinate Space:ENiagaraCoordinateSpace=::Local
    Rigid Pendulum=false
   )
   &SolveForcesAndVelocity(Clamp Velocity=false, Limit Acceleration=false)
   &PendulumConstraint()
   &CalculateAccurateVelocity()
  ]
  rendererClasses [&NiagaraMeshRendererProperties, &NiagaraSpriteRendererProperties]
  renderers [
   {rendererClass=&NiagaraMeshRendererProperties, rendererIndex=0}
   {rendererClass=&NiagaraSpriteRendererProperties, rendererIndex=1}
   {rendererClass=&NiagaraMeshRendererProperties, rendererIndex=2}
   {rendererClass=&NiagaraSpriteRendererProperties, rendererIndex=3}
  ]
 }
 Light {
  emitterUpdateScript [
   &EmitterState(
    Inactive Response=::NewEnumerator1
    Life Cycle Mode=::NewEnumerator1
    Loop Behavior=::NewEnumerator1
    Loop Duration Mode=::NewEnumerator1
    UseLoopDelay=false
   )
   &SpawnBurst_Instantaneous(Use Spawn Probability=false, UseLoopCountLimit=false)
  ]
  particleSpawnScript [&InitializeParticle(
    Color Mode=::NewEnumerator0
    Position=$User.JarOffset
    Position Mode=::NewEnumerator1
    Position Offset=<-3 -5 0>
    UsePositionOffset=true
    Write Material Random=true
   )]
  particleUpdateScript [
   &ParticleState(Kill Particles When Lifetime Has Elapsed=false)
   /SetVariables(Particles.Color=&MultiplyLinearColor(
     Linear Color=<rgba 0 9 7.7913 1>
     Scale Factor=&MakeVector4_FromFloat(Float=&Multiply_Float(
       A=$System.Timeline.Time
       B=&Waveform(
        Waveform Count=::NewEnumerator2
        "Waveform [1]"=::NewEnumerator0
        "[1] Frequency"=1
        "[1] Amplitude Min/Max"=<0.333 1>
        "[2] Average"=false
        "Waveform [2]"=::NewEnumerator1
        "[2] Input Position"=0
        "[2] Phase"=$Emitter.Age
        "[2] Phase Scale"=1
        "[2] Compound Frequency"=2
        "[2] Frequency"=2.618
        "[2] Amplitude Min/Max"=<0.666 1>
        "[2] Scale"=1
        "[2] Bias"=0
        "[2] Exponent"=1
        "[2] Pulse Duration"=0.1
        "[2] Clip Triangle Min/Max"=<0 0>
        "[2] Random Seed"=0
        "[3] Average"=false
        "[3] Input Position"=0
        "[3] Phase"=$Emitter.Age
        "[3] Phase Scale"=1
        "[3] Frequency"=3.618
        "[3] Compound Frequency"=2
        "[3] Amplitude Min/Max"=<0.833 1>
        "[3] Scale"=1
        "[3] Bias"=0
        "[3] Exponent"=1
        "[3] Pulse Duration"=0.1
        "[3] Clip Triangle Min/Max"=<0 0>
        "[3] Random Seed"=0
       )
      ))
    )) {isSetParameters=true}
  ]
  rendererClasses [&NiagaraLightRendererProperties]
  renderers [{rendererClass=&NiagaraLightRendererProperties, rendererIndex=0}]
 }
}
summary {
 systemName=NS_SwampFish_Character_Idle
 emitters [
  {emitterName=Fireflies, rendererClasses [&NiagaraMeshRendererProperties, &NiagaraSpriteRendererProperties]}
  {emitterName=Light, rendererClasses [&NiagaraLightRendererProperties]}
 ]
 userVariables [
  User.bIsFrontEnd {type=NiagaraBool}
  User.bIsFrontEndPreview {type=NiagaraBool}
  User.JarHeight {type=NiagaraFloat, defaultValue=3.25, description=""}
  User.JarOffset {type=NiagaraPosition, description="", defaultValue {}}
  User.JarRadius {type=NiagaraFloat, defaultValue=3.666, description=""}
  User.Reactivity {type=NiagaraBool}
 ]
}
