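# Material function for a glowing night-light eye on the Swamp Fish body, overriding emissive and zeroing base color and roughness on incoming material attributes.
# Fresnel smoothsteps on a transformed constant Z vector shape a color gradient with a boosted center and an added edge band, scaled by an emission control.
# It also builds a parameter-positioned UV sphere-mask highlight and rotating, tiled noise with selectable UV channel, shared through named reroutes.
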MaterialFunction /BRCosmetics/Characters/Player/Male/Medium/Bodies/M_MED_Swamp_Fish/Material/MF_NightLIghtEye
graph {nodes [
  n1 = TextureCoordinate() {coordinateIndex=1}
  AppendVector(A=@n1, B=Constant())
  n2 = GetMaterialAttributes(None=FunctionInput() {inputName=MaterialInput, inputType=FunctionInput_MaterialAttributes}) {attributeGetTypes [#1, #2, #3]}
  n3 = Constant()
  FunctionOutput(None=SetMaterialAttributes!(MaterialAttributes=@n2.MaterialAttributes, EmissiveColor=NamedRerouteUsage(), BaseColor=@n3, Roughness=@n3) {attributeSetTypes [#4, #3, #1]}) {outputName=MaterialOUT}
  Multiply(A=@n1) {constB=10}
  NamedRerouteDeclaration(None=@n2.Roughness) {name=CachedRoughness, nodeColor=<rgba 1 0 0.75 1>}
  NamedRerouteDeclaration(None=@n2.Normal) {name=CachedChangedNormal, nodeColor=<rgba 1 0.9844 0 1>}
  n4 = VectorParameter() {group=NightLight, parameterName=NightLightColor, defaultValue=<rgba 0.2591 0 0.7344 1>}
  n5 = Constant3Vector() {constant=<rgba 0 0 1 0>}
  n6 = OneMinus(None=Fresnel(Normal=Transform(None=@n5)) {baseReflectFraction=0.3, exponent=1})
  n7 = LinearInterpolate(
   A=NamedRerouteDeclaration(None=@n4.RGB) {name=CachedNightColor, nodeColor=<rgba 1 0.7031 0 1>}
   B=Multiply(A=@n4.RGB, B=ScalarParameter() {defaultValue=10, group=NightLight, parameterName=CenterIntensity}) {constB=10}
   Alpha=SmoothStep(Value=@n6) {constMax=0.8, constMin=0.3}
  )
  n8 = SmoothStep(Value=Fresnel(Normal=Transform(None=@n5)) {baseReflectFraction=0.3, exponent=1}) {constMax=0.5, constMin=0.4}
  n9 = LinearInterpolate(Alpha=OneMinus(None=@n8)) {constA=1, constB=1.5}
  n10 = OneMinus(None=SmoothStep(Value=@n6) {constMax=0.25, constMin=0.2})
  n11 = Add(
   A=Multiply(A=Multiply(A=@n7, B=@n9), B=NamedRerouteUsage())
   B=Multiply(A=Multiply(A=@n10, B=@n8), B=ScalarParameter() {defaultValue=0.2, group=NightLight, parameterName=LightEdge}) {constB=0.1}
  )
  n12 = ScalarParameter() {defaultValue=0.5, group=NightLight, parameterName=EmissionNightControl}
  n13 = NamedRerouteUsage()
  n14 = Multiply(A=NamedRerouteUsage())
  NamedRerouteDeclaration(None=Add(A=Multiply(A=@n11, B=@n12), B=Multiply(A=@n13, B=@n14))) {name=EmissiveOUT, nodeColor=<rgba 0.9219 1 0 1>}
  NamedRerouteDeclaration(None=@n2.BaseColor) {name=CachedBaseColor, nodeColor=<rgba 0 0.3672 1 1>}
  n15 = VectorParameter() {group=NightLight, parameterName=EyeHighLight_NIght, defaultValue=<rgba 1 0.62 0 1>}
  NamedRerouteDeclaration(None=SphereMask(A=TextureCoordinate(), B=ComponentMask(None=@n15.RGB) {g=true, r=true}) {attenuationRadius=0.05}) {name=NightHighLight, nodeColor=<rgba 0 1 0.5781 1>}
  n16 = LinearInterpolate(A=TextureCoordinate(), B=TextureCoordinate() {coordinateIndex=1}, Alpha=ScalarParameter() {group=NightLight, parameterName=UseUV02})
  n17 = ScalarParameter() {defaultValue=10, group=NightLight, parameterName=AnimateTilling}
  n18 = Time()
  n19 = ScalarParameter() {defaultValue=0.1, group=NightLight, parameterName=AnimateSpeed}
  n20 = TextureSampleParameter2D(UVs=Rotator(Coordinate=Multiply(A=@n16, B=@n17), Time=Multiply(A=@n18, B=@n19))) {
   group=NightLight
   parameterName=AnimateNoise
   samplerType=SAMPLERTYPE_Grayscale
   texture=&LowResBlurredNoise
  }
  NamedRerouteDeclaration(None=LinearInterpolate(Alpha=@n20.R) {constA=0.5, constB=1.2}) {name=CachedAnimatedNoise, nodeColor=<rgba 1 0 0.7969 1>}
  Transform(None=NamedRerouteUsage())
 ]}
