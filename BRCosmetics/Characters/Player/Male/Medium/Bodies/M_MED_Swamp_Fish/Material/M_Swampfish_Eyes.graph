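# Eye material for the Swamp Fish medium male cosmetic body, adding a night-light look to the shared eye shading.
# The shared eye function's attributes blend toward a night-light eye variant built from that same result.
# The blend uses a time-of-day scalar, off for morning and day and on for evening and night, replaced by a stage-switch parameter as IsFrontEnd rises.
# The blended attributes pass through the tech-art overrides function.
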Material /BRCosmetics/Characters/Player/Male/Medium/Bodies/M_MED_Swamp_Fish/Material/M_Swampfish_Eyes
graph {
 properties {
  bUseMaterialAttributes=true
  bUsedWithMeshDeformer=true
  bUsedWithSkeletalMesh=true
  pixelDepthOffsetMode=PDOM_Legacy
  shadingModel=MSM_FromMaterialExpression
  usedShadingModels=Default Lit
 }
 nodes [
  n1 = MaterialFunctionCall() {materialFunction=/Game/Characters/Shared/FNEye/Materials/MF_eyes}
  n2 = Constant()
  n3 = Constant() {r=1}
  n4 = MaterialFunctionCall(Value_Morning=@n2, Value_Day=@n2, Value_Evening=@n3, Value_Night=@n3) {materialFunction=&Scalar_Time_of_Day}
  n5 = ScalarParameter() {group=CCPM, parameterName=CCPM_StageSwitch}
  n6 = ScalarParameter() {group=CCPM, parameterName=IsFrontEnd}
  MaterialFunctionCall(In=BlendMaterialAttributes(
    A=@n1.Result
    B=MaterialFunctionCall!(MaterialInput=@n1.Result) {materialFunction=&/MF_NightLIghtEye}.MaterialOUT
    Alpha=LinearInterpolate(A=@n4.Result, B=@n5, Alpha=@n6)
   )) {materialFunction=&MF_TechArtOverrides}
 ]
}
