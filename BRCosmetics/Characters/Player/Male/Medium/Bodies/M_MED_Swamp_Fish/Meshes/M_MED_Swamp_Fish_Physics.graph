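# Physics asset for the male medium Swamp Fish player body cosmetic, covering the front and back dynamic loincloth bones and both legs.
# Each loincloth is a capsule chain from a tiny master body through five segments, with a free first joint and swing-limited, nearly twist-locked joints below.
# Thigh and calf capsules give the legs simplified shapes, joined by free knee constraints; no other body parts have bodies.
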PhysicsAsset /BRCosmetics/Characters/Player/Male/Medium/Bodies/M_MED_Swamp_Fish/Meshes/M_MED_Swamp_Fish_Physics
bodies {
 calf_l [{shapeType=Capsule, length=23.1434, radius=4.8566, shapeName=calf_l_capsule, center=<-5 0 0>, rotation=<rot 90 0 0>}]
 calf_r [{shapeType=Capsule, length=23.1434, radius=4.8566, shapeName=calf_r_capsule, center=<5 0 0>, rotation=<rot 90 0 0>}]
 dyn_loin_bk_01 [{shapeType=Capsule, length=0.505, radius=0.505, shapeName=dyn_loin_bk_01_capsule, rotation=<rot 0 -90 0>}]
 dyn_loin_bk_02 [{shapeType=Capsule, length=5.167, radius=3.7678, shapeName=dyn_loin_bk_02_capsule, center=<4 -2 0>}]
 dyn_loin_bk_03 [{shapeType=Capsule, length=5.167, radius=3.7678, shapeName=dyn_loin_bk_03_capsule, center=<4 -2 0>, rotation=<rot 90 0 0>}]
 dyn_loin_bk_04 [{shapeType=Capsule, length=5.167, radius=3.7678, shapeName=dyn_loin_bk_04_capsule, center=<4 -2 0>, rotation=<rot 90 0 0>}]
 dyn_loin_bk_05 [{shapeType=Capsule, length=5.167, radius=3.7678, shapeName=dyn_loin_bk_05_capsule, center=<4 -2 0>, rotation=<rot 90 0 0>}]
 dyn_loin_bk_MSTR [{shapeType=Capsule, length=0.505, radius=0.505, shapeName=dyn_loin_bk_MSTR_capsule, rotation=<rot 0 -90 0>}]
 dyn_loin_fr_01 [{
   shapeType=Capsule
   length=0.505
   radius=0.505
   shapeName=dyn_loin_fr_01_capsule
   center=<14.3573 0.2782 -4.7329>
   rotation=<rot 0 -92.1274 -87.969>
  }]
 dyn_loin_fr_02 [{shapeType=Capsule, length=5.167, radius=3.7678, shapeName=dyn_loin_fr_02_capsule, center=<2 -2 0>}]
 dyn_loin_fr_03 [{shapeType=Capsule, length=5.167, radius=3.7678, shapeName=dyn_loin_fr_03_capsule, center=<2 -2 0>, rotation=<rot 90 0 0>}]
 dyn_loin_fr_04 [{shapeType=Capsule, length=5.167, radius=3.7678, shapeName=dyn_loin_fr_04_capsule, center=<2 -2 0>, rotation=<rot 90 0 0>}]
 dyn_loin_fr_05 [{shapeType=Capsule, length=5.167, radius=3.7678, shapeName=dyn_loin_fr_05_capsule, center=<2 -2 0>, rotation=<rot 90 0 0>}]
 dyn_loin_fr_MSTR [{shapeType=Capsule, length=0.505, radius=0.505, shapeName=dyn_loin_fr_MSTR_capsule, rotation=<rot 0 -90 0>}]
 thigh_l [{shapeType=Capsule, length=20.2606, radius=7.7394, shapeName=thigh_l_capsule, center=<-20 0 0>, rotation=<rot 90 0 0>}]
 thigh_r [{shapeType=Capsule, length=20.2606, radius=7.7394, shapeName=thigh_r_capsule, center=<20 0 0>, rotation=<rot 90 0 0>}]
}
constraints [
 {
  bone1Name=dyn_loin_bk_01
  bone2Name=dyn_loin_bk_MSTR
  swing1LimitDegrees=45
  swing1Motion=Free
  swing2LimitDegrees=45
  swing2Motion=Free
  twistLimitDegrees=45
  twistMotion=Free
 }
 {bone1Name=dyn_loin_bk_02, bone2Name=dyn_loin_bk_01, swing1LimitDegrees=45, swing2LimitDegrees=45, twistLimitDegrees=2}
 {bone1Name=dyn_loin_bk_03, bone2Name=dyn_loin_bk_02, swing1LimitDegrees=30, swing2LimitDegrees=30, twistLimitDegrees=1}
 {bone1Name=dyn_loin_bk_04, bone2Name=dyn_loin_bk_03, swing1LimitDegrees=30, swing2LimitDegrees=30, twistLimitDegrees=1}
 {bone1Name=dyn_loin_bk_05, bone2Name=dyn_loin_bk_04, swing1LimitDegrees=30, swing2LimitDegrees=30, twistLimitDegrees=1}
 {
  bone1Name=dyn_loin_fr_01
  bone2Name=dyn_loin_fr_MSTR
  swing1LimitDegrees=45
  swing1Motion=Free
  swing2LimitDegrees=45
  swing2Motion=Free
  twistLimitDegrees=45
  twistMotion=Free
 }
 {bone1Name=dyn_loin_fr_02, bone2Name=dyn_loin_fr_01, swing1LimitDegrees=45, swing2LimitDegrees=30, twistLimitDegrees=2}
 {bone1Name=dyn_loin_fr_03, bone2Name=dyn_loin_fr_02, swing1LimitDegrees=45, swing2LimitDegrees=30, twistLimitDegrees=2}
 {bone1Name=dyn_loin_fr_04, bone2Name=dyn_loin_fr_03, swing1LimitDegrees=45, swing2LimitDegrees=30, twistLimitDegrees=2}
 {bone1Name=dyn_loin_fr_05, bone2Name=dyn_loin_fr_04, swing1LimitDegrees=45, swing2LimitDegrees=30, twistLimitDegrees=2}
 {
  bone1Name=calf_l
  bone2Name=thigh_l
  swing1LimitDegrees=45
  swing1Motion=Free
  swing2LimitDegrees=45
  swing2Motion=Free
  twistLimitDegrees=45
  twistMotion=Free
 }
 {
  bone1Name=calf_r
  bone2Name=thigh_r
  swing1LimitDegrees=45
  swing1Motion=Free
  swing2LimitDegrees=45
  swing2Motion=Free
  twistLimitDegrees=45
  twistMotion=Free
 }
]
