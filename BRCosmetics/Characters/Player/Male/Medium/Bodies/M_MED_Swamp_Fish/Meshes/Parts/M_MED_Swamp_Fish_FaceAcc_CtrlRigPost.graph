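# Post-process Control Rig for the Swamp Fish face accessory that nudges dynamic hair bones according to upper-arm pose.
# Three spherical pose readers, one on the right arm and two on the left, weight transform offsets and clavicle parent constraints, some smoothed by AlphaInterp.
# The ADS and Crouch variables shape two left-arm weights, including one driving rotation-only constraints on right-side hair.
# A separate full-weight offset adjusts the middle hair chain regardless of pose.
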ControlRigBlueprint /BRCosmetics/Characters/Player/Male/Medium/Bodies/M_MED_Swamp_Fish/Meshes/Parts/M_MED_Swamp_Fish_FaceAcc_CtrlRigPost
graphs [
 RigVMFunctionLibrary {links [], nodes []}
 RigVMModel {nodes [
   RigUnit_BeginExecution = &RigUnit_BeginExecution()
   n1 = &RigUnit_SphericalPoseReader(
    ActiveRegionScaleFactors=(PositiveWidth=3.2,NegativeWidth=0.4,PositiveHeight=4,NegativeHeight=3.5)
    ActiveRegionSize=0.1
    Debug=(bDrawDebug=false,bDrawLocalAxes=false,DebugScale=25,DebugSegments=20,DebugThickness=0.25)
    DriverAxis=(X=-1,Y=0,Z=0)
    DriverItem=(Type=Bone,Name="upperarm_r")
    FalloffRegionScaleFactors=(PositiveWidth=1,NegativeWidth=0.3,PositiveHeight=0.3,NegativeHeight=0.3)
    FalloffSize=0.45
    FlipHeightScaling=false
    FlipWidthScaling=true
    OptionalParentItem=(Type=Bone,Name="spine_05")
    RotationOffset=(X=-85,Y=-70,Z=-15)
    ExecutePin=@RigUnit_BeginExecution.ExecutePin
   )
   n2 = &RigVMFunction_AlphaInterp(InterpSpeedDecreasing=12, InterpSpeedIncreasing=15, Scale=1, bClampResult=False, Value=@n1.OutputParam)
   > &RigUnit_ModifyTransforms(
    ItemToModify="((Item=(Type=Bone,Name=\"dyn_hair_r_fr_02\"),Transform=(Rotation=(X=0,Y=0,Z=0,W=1),Translation=(X=0,Y=2,Z=2),Scale3D=(X=1,Y=1,Z=1))),(Item=(Type=Bone,Name=\"dyn_hair_r_mid_02\"),Transform=(Rotation=(X=0,Y=0,Z=0,W=1),Translation=(X=0,Y=5,Z=2),Scale3D=(X=1,Y=1,Z ...(truncated)"
    Weight=@n2.Result
    ItemToModify.0.Transform.Rotation=&RigVMFunction_MathQuaternionFromEuler(Euler=(X=0,Y=0,Z=0), RotationOrder=XYZ).Result
    ItemToModify.1.Transform.Rotation=&RigVMFunction_MathQuaternionFromEuler(Euler=(X=0,Y=0,Z=0), RotationOrder=XYZ).Result
   )
   > &RigUnit_ModifyTransforms(
    ItemToModify="((Item=(Type=Bone,Name=\"dyn_hair_04_mid\"),Transform=(Rotation=(X=0,Y=0,Z=0,W=1),Translation=(X=0,Y=-3,Z=0),Scale3D=(X=1,Y=1,Z=1))),(Item=(Type=Bone,Name=\"dyn_hair_03_mstr_mid\"),Transform=(Rotation=(X=0,Y=0,Z=0,W=1),Translation=(X=0,Y=2,Z=0),Scale3D=(X=1,Y= ...(truncated)"
    Weight=1
    WeightMinimum=0
    ItemToModify.0.Transform.Rotation=&RigVMFunction_MathQuaternionFromEuler(Euler=(X=0,Y=0,Z=-5), RotationOrder=XYZ).Result
   )
   > n3 = &RigUnit_SphericalPoseReader(
    ActiveRegionScaleFactors=(PositiveWidth=2,NegativeWidth=1.5,PositiveHeight=1.5,NegativeHeight=1)
    ActiveRegionSize=0.1
    Debug=(bDrawDebug=false,bDrawLocalAxes=false,DebugScale=25,DebugSegments=20,DebugThickness=0.25)
    DriverAxis=(X=1,Y=0,Z=0)
    DriverItem=(Type=Bone,Name="upperarm_l")
    FalloffRegionScaleFactors=(PositiveWidth=0.5,NegativeWidth=0.1,PositiveHeight=0.3,NegativeHeight=0.5)
    FalloffSize=0.45
    FlipHeightScaling=false
    FlipWidthScaling=true
    OptionalParentItem=(Type=Bone)
    RotationOffset=(X=40,Y=0,Z=-50)
   )
   > &RigUnit_ParentConstraint(
    AdvancedSettings=(InterpolationType=Average,RotationOrderForFilter=XZY)
    Child=(Type=Bone,Name="dyn_hair_l_mid_03")
    Filter=(TranslationFilter=(bX=True,bY=True,bZ=True),RotationFilter=(bX=True,bY=True,bZ=True),ScaleFilter=(bX=false,bY=false,bZ=false))
    Parents=((Item=(Type=Bone,Name="clavicle_l"),Weight=0.5))
    Weight=@n3.OutputParam
   )
   > &RigUnit_ParentConstraint(
    AdvancedSettings=(InterpolationType=Average,RotationOrderForFilter=XZY)
    Child=(Type=Bone,Name="dyn_hair_l_fr_02")
    Filter=(TranslationFilter=(bX=True,bY=True,bZ=True),RotationFilter=(bX=True,bY=True,bZ=True),ScaleFilter=(bX=false,bY=false,bZ=false))
    Parents=((Item=(Type=Bone,Name="clavicle_l"),Weight=0.5))
    Weight=@n3.OutputParam
   )
   > &RigUnit_ParentConstraint(
    AdvancedSettings=(InterpolationType=Average,RotationOrderForFilter=XZY)
    Child=(Type=Bone,Name="dyn_hair_l_mid_05")
    Filter=(TranslationFilter=(bX=True,bY=True,bZ=True),RotationFilter=(bX=True,bY=True,bZ=True),ScaleFilter=(bX=false,bY=false,bZ=false))
    Parents=((Item=(Type=Bone,Name="clavicle_l"),Weight=0.5))
    Weight=@n3.OutputParam
   )
   > &RigUnit_ModifyTransforms(
    ItemToModify="((Item=(Type=Bone,Name=\"dyn_hair_l_mid_01\"),Transform=(Rotation=(X=0,Y=0,Z=0,W=1),Translation=(X=0,Y=-2,Z=0),Scale3D=(X=1,Y=1,Z=1))),(Item=(Type=Bone,Name=\"dyn_hair_l_fr_01\"),Transform=(Rotation=(X=0,Y=0,Z=0,W=1),Translation=(X=3,Y=3,Z=-5),Scale3D=(X=1,Y=1 ...(truncated)"
    WeightMinimum=0
    ItemToModify.0.Transform.Rotation=&RigVMFunction_MathQuaternionFromEuler(Euler=(X=0,Y=0,Z=-5), RotationOrder=XYZ).Result
    ItemToModify.1.Transform.Rotation=&RigVMFunction_MathQuaternionFromEuler(Euler=(X=0,Y=0,Z=-5), RotationOrder=XYZ).Result
    Weight=@n3.OutputParam
   )
   n4 = &RigVMFunction_AlphaInterp(
    InterpSpeedDecreasing=10
    InterpSpeedIncreasing=10
    bClampResult=true
    Value=@n3.OutputParam
    Scale=RigVMVariableNode() {variable=Crouch}.Value
    ClampMax=RigVMVariableNode() {variable=ADS}.Value
   )
   > &RigUnit_ModifyTransforms(
    ItemToModify="((Item=(Type=Bone,Name=\"dyn_hair_l_fr_01\"),Transform=(Rotation=(X=0,Y=0,Z=0,W=1),Translation=(X=0,Y=0,Z=0),Scale3D=(X=1,Y=1,Z=1))),(Item=(Type=Bone,Name=\"dyn_hair_l_mid_01\"),Transform=(Rotation=(X=0,Y=0,Z=0,W=1),Translation=(X=0,Y=0,Z=-5),Scale3D=(X=1,Y=1, ...(truncated)"
    WeightMinimum=0
    Weight=@n4.Result
    ItemToModify.1.Transform.Rotation=&RigVMFunction_MathQuaternionFromEuler(Euler=(X=0,Y=0,Z=0), RotationOrder=XYZ).Result
    ItemToModify.0.Transform.Rotation=&RigVMFunction_MathQuaternionFromEuler(Euler=(X=0,Y=0,Z=0), RotationOrder=XYZ).Result
   )
   > n5 = &RigUnit_SphericalPoseReader(
    ActiveRegionScaleFactors=(PositiveWidth=2.5,NegativeWidth=2,PositiveHeight=2.5,NegativeHeight=1)
    ActiveRegionSize=0.1
    Debug=(bDrawDebug=false,bDrawLocalAxes=false,DebugScale=25,DebugSegments=20,DebugThickness=0.25)
    DriverAxis=(X=1,Y=0,Z=0)
    DriverItem=(Type=Bone,Name="upperarm_l")
    FalloffRegionScaleFactors=(PositiveWidth=0.5,NegativeWidth=0.4,PositiveHeight=0.4,NegativeHeight=0.2)
    FalloffSize=0.45
    FlipHeightScaling=false
    FlipWidthScaling=true
    OptionalParentItem=(Type=Bone,Name="spine_05")
    RotationOffset=(X=120,Y=0,Z=-90)
   )
   AlphaInterp = &RigVMFunction_AlphaInterp(InterpSpeedDecreasing=5, InterpSpeedIncreasing=10, bClampResult=False, Value=@n5.OutputParam, Scale=RigVMVariableNode() {variable=ADS}.Value)
   > &RigUnit_ParentConstraint(
    AdvancedSettings=(InterpolationType=Average,RotationOrderForFilter=XZY)
    Child=(Type=Bone,Name="dyn_hair_r_fr_03")
    Filter=(TranslationFilter=(bX=false,bY=false,bZ=false),RotationFilter=(bX=True,bY=True,bZ=True),ScaleFilter=(bX=false,bY=false,bZ=false))
    Parents=((Item=(Type=Bone,Name="clavicle_r"),Weight=0.5))
    Weight=@AlphaInterp.Result
   )
   > &RigUnit_ParentConstraint(
    AdvancedSettings=(InterpolationType=Average,RotationOrderForFilter=XZY)
    Child=(Type=Bone,Name="dyn_hair_r_mid_04")
    Filter=(TranslationFilter=(bX=false,bY=false,bZ=false),RotationFilter=(bX=True,bY=True,bZ=True),ScaleFilter=(bX=false,bY=false,bZ=false))
    Parents=((Item=(Type=Bone,Name="clavicle_r"),Weight=0.5))
    Weight=@AlphaInterp.Result
   )
   > &RigUnit_ModifyTransforms(
    ItemToModify="((Item=(Type=None,Name=\"None\"),Transform=(Rotation=(X=0,Y=0,Z=0,W=1),Translation=(X=0,Y=-2,Z=0),Scale3D=(X=1,Y=1,Z=1))),(Item=(Type=None,Name=\"None\"),Transform=(Rotation=(X=0,Y=0,Z=0,W=1),Translation=(X=3,Y=3,Z=-5),Scale3D=(X=1,Y=1,Z=1))),(Item=(Type=Bone, ...(truncated)"
    WeightMinimum=0
    Weight=@AlphaInterp.Result
   )
  ]}
]
hierarchy [
 root {loc=<0 0 0>, rot=<0 0 0 1>}
 pelvis {parent=root, loc=<0 -0.7071 93.1191>, rot=<0.0127 0.707 -0.0127 -0.707>}
 spine_01 {parent=pelvis, loc=<6.7494 -0.6388 0>, rot=<0 0 0.034 -0.9994>}
 spine_02 {parent=spine_01, loc=<7.2518 -0.035 0>, rot=<0 0 0.0238 -0.9997>}
 spine_03 {parent=spine_02, loc=<6.0104 0 0>, rot=<0 0 0.0149 -0.9999>}
 spine_04 {parent=spine_03, loc=<8.1024 0 0>, rot=<0 0 -0.0132 -0.9999>}
 spine_05 {parent=spine_04, loc=<10.0146 0 0>, rot=<0 0 -0.0015 -1>}
 neck_01 {parent=spine_05, loc=<11.3138 0.1529 0>, rot=<0 0 -0.2003 -0.9797>}
 neck_02 {parent=neck_01, loc=<4.5627 0 0>, rot=<0 0 0.0504 -0.9987>}
 head {parent=neck_02, loc=<5.3914 0 0>, rot=<0 0 0.115 -0.9934>}
 dyn_strand_fr_r_out_01 {parent=head, loc=<4.1439 18.7301 6.7141>, rot=<-0.1172 -0.9927 -0.0037 -0.0292>}
 dyn_strand_fr_r_out_02 {parent=dyn_strand_fr_r_out_01, loc=<3.0049 0.4079 0.1511>, rot=<-0.0089 -0.0619 0.1884 -0.9801>}
 dyn_strand_fr_r_out_03 {parent=dyn_strand_fr_r_out_02, loc=<3.2467 0.0107 -0.0091>, rot=<0 0 0 1>}
 dyn_strand_fr_r_out_04 {parent=dyn_strand_fr_r_out_03, loc=<4.3041 0.0143 -0.0121>, rot=<0 0 0 1>}
 dyn_strand_fr_r_out_05 {parent=dyn_strand_fr_r_out_04, loc=<5.0982 0.0223 -0.019>, rot=<0 0 0 1>}
 dyn_strand_fr_r_in_01 {parent=head, loc=<4.4912 19.8559 3.9967>, rot=<-0.1149 -0.9933 -0.0009 -0.0075>}
 dyn_strand_fr_r_in_02 {parent=dyn_strand_fr_r_in_01, loc=<3.767 0.2542 -0.1634>, rot=<0.0004 0.0098 0.2005 -0.9796>}
 dyn_strand_fr_r_in_03 {parent=dyn_strand_fr_r_in_02, loc=<3.7839 -0.0001 0>, rot=<0 0 0 1>}
 dyn_strand_fr_r_in_04 {parent=dyn_strand_fr_r_in_03, loc=<4.8023 -0.0001 0>, rot=<0 0 0 1>}
 dyn_strand_fr_r_in_05 {parent=dyn_strand_fr_r_in_04, loc=<6.0105 0.0002 0>, rot=<0 0 0 1>}
 dyn_strand_fr_l_01 {parent=head, loc=<4.2426 18.4948 -6.8157>, rot=<-0.0817 -0.9965 -0.0018 -0.0197>}
 dyn_strand_fr_l_02 {parent=dyn_strand_fr_l_01, loc=<4.6891 0 0>, rot=<-0.0064 -0.1412 0.125 -0.982>}
 dyn_strand_fr_l_in_01 {parent=dyn_strand_fr_l_02, loc=<0.3995 0.4473 -0.6517>, rot=<-0.005 0.1334 -0.0437 -0.9901>}
 dyn_strand_fr_l_in_02 {parent=dyn_strand_fr_l_in_01, loc=<4.1223 -0.0223 -0.0761>, rot=<0 0 0 1>}
 dyn_strand_fr_l_in_03 {parent=dyn_strand_fr_l_in_02, loc=<4.7472 -0.0258 -0.0876>, rot=<0 0 0 1>}
 dyn_strand_fr_l_out_01 {parent=dyn_strand_fr_l_02, loc=<1.11 -0.7172 0.9183>, rot=<-0.0112 0.1621 -0.0119 -0.9866>}
 dyn_strand_fr_l_out_02 {parent=dyn_strand_fr_l_out_01, loc=<5.9551 0 0>, rot=<0 0 0 1>}
 dyn_strand_fr_l_out_03 {parent=dyn_strand_fr_l_out_02, loc=<9.4243 0 0>, rot=<0 0 0 1>}
 dyn_strand_fr_l_out_04 {parent=dyn_strand_fr_l_out_03, loc=<7.7067 0 0>, rot=<0 0 0 1>}
 dyn_strand_fr_l_mid_01 {parent=dyn_strand_fr_l_01, loc=<5.639 -1.2362 1.4763>, rot=<-0.0027 0.0912 0.1998 -0.9756>}
 dyn_strand_fr_l_mid_02 {parent=dyn_strand_fr_l_mid_01, loc=<4.9726 -0.0001 0>, rot=<0 0 0 1>}
 dyn_strand_fr_l_mid_03 {parent=dyn_strand_fr_l_mid_02, loc=<5.2915 0.0001 0>, rot=<0 0 0 1>}
 dyn_strand_fr_l_mid_04 {parent=dyn_strand_fr_l_mid_03, loc=<6.2544 0 0>, rot=<0 0 0 1>}
 dyn_hair_r_fr_01 {parent=head, loc=<4.1128 1.9981 13.5526>, rot=<-0.1417 -0.7412 -0.6557 -0.0229>}
 dyn_hair_r_fr_02 {parent=dyn_hair_r_fr_01, loc=<8.8004 -0.2045 0.0259>, rot=<0.0079 0.0036 0.0305 -0.9995>}
 dyn_hair_r_fr_03 {parent=dyn_hair_r_fr_02, loc=<9.8781 -0.3681 0.0408>, rot=<0.011 0.0042 0.0424 -0.999>}
 dyn_hair_r_fr_04 {parent=dyn_hair_r_fr_03, loc=<6.7388 -0.3007 0.0267>, rot=<0.0111 0.0034 0.0433 -0.999>}
 dyn_hair_r_mid_01 {parent=head, loc=<1.7451 -1.0702 13.3227>, rot=<-0.0355 0.6595 0.745 0.0932>}
 dyn_hair_r_mid_02 {parent=dyn_hair_r_mid_01, loc=<8.7918 0.5561 0.5834>, rot=<-0.0302 0.1252 -0.127 -0.9835>}
 dyn_hair_r_mid_03 {parent=dyn_hair_r_mid_02, loc=<8.3385 0.2705 0.2334>, rot=<-0.0237 0.042 -0.0588 -0.9971>}
 dyn_hair_r_mid_04 {parent=dyn_hair_r_mid_03, loc=<6.9066 0.021 0.1069>, rot=<-0.0095 -0.0308 -0.0289 -0.9991>}
 dyn_hair_r_mid_05 {parent=dyn_hair_r_mid_04, loc=<7.5965 0.1241 -0.6833>, rot=<-0.0014 -0.088 -0.0159 -0.996>}
 dyn_hair_l_fr_01 {parent=head, loc=<-1.1251 -0.8252 -13.2566>, rot=<-0.1424 -0.6727 0.7129 0.1377>}
 dyn_hair_l_fr_02 {parent=dyn_hair_l_fr_01, loc=<6.6287 0 0>, rot=<0 0 0 1>}
 dyn_hair_l_fr_03 {parent=dyn_hair_l_fr_02, loc=<6.6149 0 0>, rot=<0 0 0 1>}
 dyn_hair_l_mid_01 {parent=head, loc=<-1.9976 -2.1109 -12.8196>, rot=<-0.2337 -0.6172 0.7467 0.0825>}
 dyn_hair_l_mid_02 {parent=dyn_hair_l_mid_01, loc=<5.3125 -0.0222 0.0074>, rot=<-0.0043 0.0043 0.0126 -0.9999>}
 dyn_hair_l_mid_03 {parent=dyn_hair_l_mid_02, loc=<5.5734 -0.0868 0.0299>, rot=<-0.012 0.0126 0.0353 -0.9992>}
 dyn_hair_l_mid_04 {parent=dyn_hair_l_mid_03, loc=<6.3012 -0.1963 0.0729>, rot=<-0.0165 0.0195 0.05 -0.9984>}
 dyn_hair_l_mid_05 {parent=dyn_hair_l_mid_04, loc=<6.1015 -0.1698 0.0696>, rot=<-0.0101 0.0134 0.0318 -0.9994>}
 dyn_tear_l_01 {parent=head, loc=<12.4097 9.7157 -7.6493>, rot=<0.5966 -0.4413 -0.2206 -0.633>}
 dyn_tear_l_02 {parent=dyn_tear_l_01, loc=<2.7487 0 0>, rot=<0 0 0 1>}
 dyn_tear_r_01 {parent=head, loc=<11.8336 1.9646 10.153>, rot=<0.3431 -0.7607 0.2853 0.4714>}
 dyn_tear_r_02 {parent=dyn_tear_r_01, loc=<3.5854 0 0.0003>, rot=<0 0 0 1>}
 dyn_tear_r_03 {parent=dyn_tear_r_02, loc=<3.5009 0 -0.0003>, rot=<0 0 0 1>}
 dyn_hat_tie_01 {parent=head, loc=<11.8176 -8.2001 2.5026>, rot=<0.1966 0.0372 0.9751 -0.0957>}
 dyn_hat_tie_02 {parent=dyn_hat_tie_01, loc=<4.1607 0 0>, rot=<-0.0131 -0.1236 0.0119 -0.9922>}
 dyn_hat_tie_03 {parent=dyn_hat_tie_02, loc=<5.5742 -0.0267 -0.2432>, rot=<-0.0128 -0.0755 0.0093 -0.997>}
 dyn_hat_tie_04 {parent=dyn_hat_tie_03, loc=<4.3115 -0.0007 -0.0051>, rot=<-0.005 -0.0297 0.0042 -0.9995>}
 dyn_hat_bow_01 {parent=head, loc=<9.8789 -8.1271 -1.3486>, rot=<-0.1428 -0.0512 0.9654 -0.2121>}
 dyn_hat_bow_02 {parent=dyn_hat_bow_01, loc=<3.7061 0 0>, rot=<0 0 0 1>}
 dyn_hat_bow_03 {parent=dyn_hat_bow_02, loc=<3.9277 0.0001 0>, rot=<0 0 0 1>}
 dyn_hat_fr_01 {parent=head, loc=<15.4991 -0.3604 -2.9336>, rot=<0.9138 -0.2342 -0.2364 -0.2327>}
 dyn_hat_fr_02 {parent=dyn_hat_fr_01, loc=<3.4116 0 0>, rot=<0 0 0 1>}
 dyn_hat_fr_03 {parent=dyn_hat_fr_02, loc=<2.1908 0 0>, rot=<0 0 0 1>}
 dyn_hat_bk_01 {parent=head, loc=<15.1996 -2.2239 0.0993>, rot=<0.9079 -0.314 -0.0304 -0.2762>}
 dyn_hat_bk_02 {parent=dyn_hat_bk_01, loc=<3.47 0 0>, rot=<0 0 0 1>}
 dyn_hat_bk_03 {parent=dyn_hat_bk_02, loc=<1.8963 0 0>, rot=<0 0 0 1>}
 dyn_hair_r_MSTR {parent=head, loc=<10.0791 -1.0364 6.0327>, rot=<0.1315 -0.5972 0.7911 0.0152>}
 dyn_hair_r_00 {parent=dyn_hair_r_MSTR, loc=<0.5 0 0>, rot=<0.1445 0.0023 -0.0046 -0.9895>}
 dyn_hair_r_01 {parent=dyn_hair_r_00, loc=<6.1151 0 0>, rot=<-0.011 0.0419 -0.063 -0.9971>}
 dyn_hair_r_02 {parent=dyn_hair_r_01, loc=<7.6839 0 0>, rot=<0.0327 0.2728 0.061 -0.9596>}
 dyn_hair_r_03 {parent=dyn_hair_r_02, loc=<7.4404 0 0>, rot=<-0.0227 -0.1996 -0.0356 -0.979>}
 dyn_hair_r_04 {parent=dyn_hair_r_03, loc=<8.1878 0 0>, rot=<0 -0.1 0 -0.995>}
 dyn_hair_r_05 {parent=dyn_hair_r_04, loc=<9.8624 0.8728 -0.4626>, rot=<-0.0685 -0.1063 -0.0644 -0.9899>}
 dyn_hair_l_MSTR {parent=head, loc=<8.8541 0.6916 -6.499>, rot=<0.0262 -0.7919 0.5984 -0.1187>}
 dyn_hair_l_00 {parent=dyn_hair_l_MSTR, loc=<0.5 0 0>, rot=<-0.1423 -0.0479 0.0091 -0.9886>}
 dyn_hair_l_01 {parent=dyn_hair_l_00, loc=<7.2513 0 0>, rot=<-0.0077 0.1474 -0.0325 -0.9885>}
 dyn_hair_l_02 {parent=dyn_hair_l_01, loc=<7.2772 0 0>, rot=<0.0148 0.2066 0.0222 -0.9781>}
 dyn_hair_l_03 {parent=dyn_hair_l_02, loc=<8.2624 0 0>, rot=<0.0067 -0.2123 0.0102 -0.9771>}
 dyn_hair_l_04 {parent=dyn_hair_l_03, loc=<8.9834 0 0>, rot=<0.0001 -0.1438 0.0605 -0.9878>}
 dyn_hair_l_05 {parent=dyn_hair_l_04, loc=<9.7831 0 0.0463>, rot=<0 -0.0488 0 -0.9988>}
 dyn_hair_bk_MSTR {parent=head, loc=<11.6835 -5.2363 0>, rot=<0.1215 -0.6966 0.6966 -0.1215>}
 dyn_hair_bk_00 {parent=dyn_hair_bk_MSTR, loc=<0.5 0 0>, rot=<0 -0.0572 0 -0.9984>}
 dyn_hair_bk_01 {parent=dyn_hair_bk_00, loc=<6.0938 0 0>, rot=<0 -0.0272 0 -0.9996>}
 dyn_hair_bk_02 {parent=dyn_hair_bk_01, loc=<8.7854 0 0>, rot=<0 0.1593 0 -0.9872>}
 dyn_hair_bk_03 {parent=dyn_hair_bk_02, loc=<9.4823 0 0>, rot=<0 -0.1683 0 -0.9857>}
 dyn_hair_bk_04 {parent=dyn_hair_bk_03, loc=<10.3536 0 0>, rot=<0 -0.0798 0 -0.9968>}
 dyn_hair_bk_05 {parent=dyn_hair_bk_04, loc=<11.2711 0 0>, rot=<0 0 0 1>}
 dyn_hair_l_bk_01 {parent=head, loc=<3.3319 0.6484 -0.1313>, rot=<-0.3255 0.0591 -0.1783 -0.9267>}
 dyn_hair_l_bk_02 {parent=dyn_hair_l_bk_01, loc=<7.8474 0 0>, rot=<0 0 0 1>}
 dyn_hair_l_bk_03 {parent=dyn_hair_l_bk_02, loc=<5.3765 0 0>, rot=<0 0 0 1>}
 dyn_hair_l_bk_04 {parent=dyn_hair_l_bk_03, loc=<4.3787 0 0>, rot=<0 0 0 1>}
 dyn_hair_r_bk_01 {parent=head, loc=<3.8041 -0.4433 -0.4476>, rot=<0.9015 -0.1105 -0.2254 0.3526>}
 dyn_hair_r_bk_02 {parent=dyn_hair_r_bk_01, loc=<4.354 0 0>, rot=<0 0 0 1>}
 dyn_hair_r_bk_03 {parent=dyn_hair_r_bk_02, loc=<4.4818 0 0>, rot=<0 0 0 1>}
 dyn_hair_back_mstr_r {parent=head, loc=<9.1482 -4.0009 5.564>, rot=<0.4999 -0.5 0.5042 0.4958>}
 dyn_hair_back_01_r {parent=dyn_hair_back_mstr_r, loc=<0 0 0>, rot=<-0.0709 0.0447 -0.7005 -0.7087>}
 dyn_hair_back_02_r {parent=dyn_hair_back_01_r, loc=<12.6565 0 0>, rot=<-0.1642 0.2607 0.0756 -0.9483>}
 dyn_hair_back_03_mstr_r {parent=dyn_hair_back_02_r, loc=<12.006 0 0>, rot=<0.0859 0 0 -0.9963>}
 dyn_hair_back_03_r {parent=dyn_hair_back_03_mstr_r, loc=<0 0 0>, rot=<0.0289 -0.1516 0.017 -0.9879>}
 dyn_hair_back_04_r {parent=dyn_hair_back_03_r, loc=<12.0961 0 0>, rot=<0.0019 -0.0609 0.044 -0.9972>}
 dyn_hair_back_05_r {parent=dyn_hair_back_04_r, loc=<12.042 0 0>, rot=<-0.0521 -0.1127 -0.0097 -0.9922>}
 dyn_hair_back_04_fr {parent=dyn_hair_back_03_r, loc=<12.9856 -3.3695 -1.1769>, rot=<-0.0924 -0.1317 0.1173 -0.98>}
 dyn_hair_back_05_fr {parent=dyn_hair_back_04_fr, loc=<9.5607 0 0>, rot=<-0.0796 -0.3274 0.0935 -0.9369>}
 dyn_hair_mid_mstr {parent=head, loc=<9.5706 -5.4997 0>, rot=<0.4999 -0.5 0.5042 0.4958>}
 dyn_hair_01_mid {parent=dyn_hair_mid_mstr, loc=<0 0 0>, rot=<-0.0636 0.0612 -0.7012 -0.7075>}
 dyn_hair_02_mid {parent=dyn_hair_01_mid, loc=<12.379 0 0>, rot=<0.0116 0.1243 0.0103 -0.9921>}
 dyn_hair_03_mstr_mid {parent=dyn_hair_02_mid, loc=<11.7877 0 0>, rot=<0 0 0 1>}
 dyn_hair_03_mid {parent=dyn_hair_03_mstr_mid, loc=<0 0 0>, rot=<-0.0005 -0.0178 0.0001 -0.9998>}
 dyn_hair_04_mid {parent=dyn_hair_03_mid, loc=<12.2089 0 0>, rot=<-0.0031 -0.0797 0.0004 -0.9968>}
 dyn_hair_05_mid {parent=dyn_hair_04_mid, loc=<12.2908 0 0>, rot=<-0.0236 -0.0912 0.0165 -0.9954>}
 dyn_hair_back_mstr_l {parent=head, loc=<9.1487 -4.0009 -5.564>, rot=<0.4999 -0.5 0.5042 0.4958>}
 dyn_hair_back_01_l {parent=dyn_hair_back_mstr_l, loc=<0 0 0>, rot=<-0.0762 0.1155 -0.6999 -0.7007>}
 dyn_hair_back_02_l {parent=dyn_hair_back_01_l, loc=<12.6412 0 0>, rot=<0 0.0637 0 -0.998>}
 dyn_hair_back_03_mstr_l {parent=dyn_hair_back_02_l, loc=<11.9654 0 0>, rot=<0 0 0 1>}
 dyn_hair_back_03_l {parent=dyn_hair_back_03_mstr_l, loc=<0 0 0>, rot=<0.0497 -0.1259 0.0181 -0.9906>}
 dyn_hair_back_04_l {parent=dyn_hair_back_03_l, loc=<12.0147 0 0>, rot=<0.0423 -0.092 -0.0316 -0.9944>}
 dyn_hair_back_05_l {parent=dyn_hair_back_04_l, loc=<11.9574 0 0>, rot=<0.0196 -0.0108 -0.0098 -0.9997>}
 dyn_hair_back_04_fl {parent=dyn_hair_back_03_l, loc=<13.047 4.2082 -0.7543>, rot=<0.1736 -0.0723 -0.1422 -0.9718>}
 dyn_hair_back_05_fl {parent=dyn_hair_back_04_fl, loc=<9.6826 0 0>, rot=<-0.033 -0.108 0.0124 -0.9935>}
 dyn_fin_bk_c_top {parent=neck_01, loc=<-4.3982 -5.7679 -0.1387>, rot=<0.5249 -0.8504 0.0324 0.0145>}
 clavicle_l {parent=spine_05, loc=<4.3575 -0.1466 -2.3223>, rot=<0.0465 -0.7289 -0.009 -0.6829>}
 upperarm_l {parent=clavicle_l, loc=<15.5111 0 0>, rot=<0.0226 -0.3234 -0.0088 -0.9459>}
 clavicle_r {parent=spine_05, loc=<4.3575 -0.1466 2.3223>, rot=<-0.6606 -0.0454 0.7493 -0.0134>}
 upperarm_r {parent=clavicle_r, loc=<-15.5111 0 0>, rot=<0.0226 -0.3234 -0.0088 -0.9459>}
 dyn_fin_bk_c_mid {parent=spine_05, loc=<-2.6757 -10.2788 -0.0668>, rot=<0.5335 -0.8458 0.0079 -0.001>}
 dyn_fin_bk_c_bot {parent=spine_03, loc=<4.3652 -10.398 -0.0549>, rot=<0.6297 -0.7768 0.0088 -0.0051>}
]
variables [
 ADS {type=double, default=0}
 Crouch {type=double, default=0}
]
