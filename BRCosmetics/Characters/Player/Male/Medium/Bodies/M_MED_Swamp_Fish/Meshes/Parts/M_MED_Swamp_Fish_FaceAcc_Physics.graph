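# Physics asset for the face accessory mesh of the male medium Swamp Fish player body, rigging its dyn_hair bone chains as capsule bodies alongside upper-body capsules.
# Four side chains hang from the head with narrow twist limits; three rear chains (left, middle, right) start at tiny master bodies and swing wider.
# Angularly free constraints link the torso, arms and chain roots, plus extra master-to-spine_05 and rear left/right-to-middle pairs.
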PhysicsAsset /BRCosmetics/Characters/Player/Male/Medium/Bodies/M_MED_Swamp_Fish/Meshes/Parts/M_MED_Swamp_Fish_FaceAcc_Physics
bodies {
 clavicle_l [{shapeType=Capsule, length=37.9854, radius=7.903, shapeName=clavicle_l_capsule, center=<1.0042 -0.5 -3.3488>, rotation=<rot 86 0 0>}]
 clavicle_r [{
   shapeType=Capsule
   length=17.7354
   radius=6.903
   shapeName=clavicle_r_capsule
   center=<-6.4257 -0.3013 1.2132>
   rotation=<rot 86 -0.0023 177.3418>
  }]
 dyn_hair_03_mid [{shapeType=Capsule, length=7.5, radius=6.5, shapeName=dyn_hair_03_mid_capsule, center=<4 0 4>, rotation=<rot 90 0 0>}]
 dyn_hair_03_mstr_mid [{shapeType=Capsule, length=0.505, radius=0.505, shapeName=dyn_hair_03_mstr_mid_capsule, rotation=<rot 0 -90 0>}]
 dyn_hair_04_mid [{shapeType=Capsule, length=7.5, radius=6.5, shapeName=dyn_hair_04_mid_capsule, center=<4 0 4>, rotation=<rot 90 0 0>}]
 dyn_hair_05_mid [{shapeType=Capsule, length=7.5, radius=6.5, shapeName=dyn_hair_05_mid_capsule, center=<4 0 4>, rotation=<rot 90 0 0>}]
 dyn_hair_back_03_l [{shapeType=Capsule, length=8.25, radius=5.75, shapeName=dyn_hair_back_03_l_capsule, center=<4 0 4>, rotation=<rot 90 0 0>}]
 dyn_hair_back_03_mstr_l [{shapeType=Capsule, length=0.505, radius=0.505, shapeName=dyn_hair_back_03_mstr_l_capsule, rotation=<rot 0 -90 0>}]
 dyn_hair_back_03_mstr_r [{shapeType=Capsule, length=0.505, radius=0.505, shapeName=dyn_hair_back_03_mstr_r_capsule, rotation=<rot 0 -90 0>}]
 dyn_hair_back_03_r [{shapeType=Capsule, length=8.25, radius=5.75, shapeName=dyn_hair_back_03_r_capsule, center=<3 0 5>, rotation=<rot 90 0 0>}]
 dyn_hair_back_04_l [{shapeType=Capsule, length=7.625, radius=6.375, shapeName=dyn_hair_back_04_l_capsule, center=<4 0 5>, rotation=<rot 90 0 0>}]
 dyn_hair_back_04_r [{shapeType=Capsule, length=7.5, radius=6.5, shapeName=dyn_hair_back_04_r_capsule, center=<3 0 5>, rotation=<rot 90 0 0>}]
 dyn_hair_back_05_l [{shapeType=Capsule, length=7.625, radius=6.375, shapeName=dyn_hair_back_05_l_capsule, center=<4 0 5.5>, rotation=<rot 90 0 0>}]
 dyn_hair_back_05_r [{shapeType=Capsule, length=7.5, radius=6.5, shapeName=dyn_hair_back_05_r_capsule, center=<3 0 5>, rotation=<rot 90 0 0>}]
 dyn_hair_l_fr_01 [{shapeType=Capsule, length=5, radius=3, shapeName=dyn_hair_l_fr_01_capsule, center=<3 0 0>, rotation=<rot 90 0 0>}]
 dyn_hair_l_fr_02 [{shapeType=Capsule, length=5, radius=3, shapeName=dyn_hair_l_fr_02_capsule, center=<5 0 -2>, rotation=<rot 90 0 0>}]
 dyn_hair_l_fr_03 [{shapeType=Capsule, length=5, radius=3, shapeName=dyn_hair_l_fr_03_capsule, center=<5 0 -2>, rotation=<rot 90 0 0>}]
 dyn_hair_l_mid_01 [{shapeType=Capsule, length=5, radius=3, shapeName=dyn_hair_l_mid_01_capsule, center=<3 0 0>, rotation=<rot 90 0 0>}]
 dyn_hair_l_mid_02 [{shapeType=Capsule, length=5, radius=3, shapeName=dyn_hair_l_mid_02_capsule, center=<5 0 -2>, rotation=<rot 90 0 0>}]
 dyn_hair_l_mid_03 [{shapeType=Capsule, length=5, radius=3, shapeName=dyn_hair_l_mid_03_capsule, center=<5 0 -2>, rotation=<rot 90 0 0>}]
 dyn_hair_l_mid_04 [{shapeType=Capsule, length=5, radius=3, shapeName=dyn_hair_l_mid_04_capsule, center=<5 0 -2>, rotation=<rot 90 0 0>}]
 dyn_hair_l_mid_05 [{shapeType=Capsule, length=5, radius=3, shapeName=dyn_hair_l_mid_05_capsule, center=<5 0 -2>, rotation=<rot 90 0 0>}]
 dyn_hair_r_fr_01 [{shapeType=Capsule, length=5, radius=3, shapeName=dyn_hair_r_fr_01_capsule, center=<3 0 0>, rotation=<rot 90 0 0>}]
 dyn_hair_r_fr_02 [{shapeType=Capsule, length=5.5, radius=2.5, shapeName=dyn_hair_r_fr_02_capsule, center=<5 0 2>, rotation=<rot 90 0 0>}]
 dyn_hair_r_fr_03 [{shapeType=Capsule, length=5.5, radius=2.5, shapeName=dyn_hair_r_fr_03_capsule, center=<5 0 2>, rotation=<rot 90 0 0>}]
 dyn_hair_r_fr_04 [{shapeType=Capsule, length=5.5, radius=2.5, shapeName=dyn_hair_r_fr_04_capsule, center=<5 0 2>, rotation=<rot 90 0 0>}]
 dyn_hair_r_mid_01 [{shapeType=Capsule, length=5, radius=3, shapeName=dyn_hair_r_mid_01_capsule, center=<3 0 0>, rotation=<rot 90 0 0>}]
 dyn_hair_r_mid_02 [{shapeType=Capsule, length=5, radius=2, shapeName=dyn_hair_r_mid_02_capsule, center=<5 0 2>, rotation=<rot 90 0 0>}]
 dyn_hair_r_mid_03 [{shapeType=Capsule, length=5, radius=2, shapeName=dyn_hair_r_mid_03_capsule, center=<5 0 2>, rotation=<rot 90 0 0>}]
 dyn_hair_r_mid_04 [{shapeType=Capsule, length=5, radius=2, shapeName=dyn_hair_r_mid_04_capsule, center=<5 0 2>, rotation=<rot 90 0 0>}]
 dyn_hair_r_mid_05 [{shapeType=Capsule, length=5, radius=2, shapeName=dyn_hair_r_mid_05_capsule, center=<5 0 2>, rotation=<rot 90 0 0>}]
 head [{shapeType=Capsule, length=6.1, radius=6.5, shapeName=head_capsule, center=<2.6 2.5 0>, rotation=<rot 90 0 0>}]
 neck_01 [{shapeType=Capsule, length=11.0557, radius=10.2668, shapeName=neck_01_capsule, center=<2.6686 3.0164 0>, rotation=<rot 90 160.2328 170.2331>}]
 spine_02 [{shapeType=Capsule, length=46.0218, radius=14.8093, shapeName=spine_02_capsule, center=<0 0.7989 0>}]
 spine_04 [{shapeType=Capsule, length=48.7919, radius=13.3093, shapeName=spine_04_capsule, center=<1 0.7857 0>}]
 spine_05 [{shapeType=Capsule, length=39.9113, radius=9.2199, shapeName=spine_05_capsule, center=<-0.5896 0.147 0>}]
 upperarm_l [{shapeType=Capsule, length=29.0812, radius=7.1347, center=<10.8056 -0.4936 0.3456>, rotation=<rot 90 0 0>}]
 upperarm_r [{shapeType=Capsule, length=29.0812, radius=7.1347, center=<-10.8056 0.4936 3.1281>, rotation=<rot 90 0 0>}]
}
constraints [
 {
  bone1Name=spine_04
  bone2Name=spine_02
  swing1LimitDegrees=45
  swing1Motion=Free
  swing2LimitDegrees=45
  swing2Motion=Free
  twistLimitDegrees=45
  twistMotion=Free
 }
 {
  bone1Name=spine_05
  bone2Name=spine_04
  swing1LimitDegrees=45
  swing1Motion=Free
  swing2LimitDegrees=45
  swing2Motion=Free
  twistLimitDegrees=45
  twistMotion=Free
 }
 {
  bone1Name=clavicle_l
  bone2Name=spine_05
  swing1LimitDegrees=45
  swing1Motion=Free
  swing2LimitDegrees=45
  swing2Motion=Free
  twistLimitDegrees=45
  twistMotion=Free
 }
 {
  bone1Name=upperarm_l
  bone2Name=clavicle_l
  swing1LimitDegrees=45
  swing1Motion=Free
  swing2LimitDegrees=45
  swing2Motion=Free
  twistLimitDegrees=45
  twistMotion=Free
 }
 {
  bone1Name=clavicle_r
  bone2Name=spine_05
  swing1LimitDegrees=45
  swing1Motion=Free
  swing2LimitDegrees=45
  swing2Motion=Free
  twistLimitDegrees=45
  twistMotion=Free
 }
 {
  bone1Name=upperarm_r
  bone2Name=clavicle_r
  swing1LimitDegrees=45
  swing1Motion=Free
  swing2LimitDegrees=45
  swing2Motion=Free
  twistLimitDegrees=45
  twistMotion=Free
 }
 {
  bone1Name=neck_01
  bone2Name=spine_05
  swing1LimitDegrees=45
  swing1Motion=Free
  swing2LimitDegrees=45
  swing2Motion=Free
  twistLimitDegrees=45
  twistMotion=Free
 }
 {
  bone1Name=head
  bone2Name=neck_01
  swing1LimitDegrees=45
  swing1Motion=Free
  swing2LimitDegrees=45
  swing2Motion=Free
  twistLimitDegrees=45
  twistMotion=Free
 }
 {
  bone1Name=dyn_hair_l_fr_01
  bone2Name=head
  swing1LimitDegrees=45
  swing1Motion=Free
  swing2LimitDegrees=45
  swing2Motion=Free
  twistLimitDegrees=45
  twistMotion=Free
 }
 {bone1Name=dyn_hair_l_fr_02, bone2Name=dyn_hair_l_fr_01, swing1LimitDegrees=15, swing2LimitDegrees=30, twistLimitDegrees=2}
 {bone1Name=dyn_hair_l_fr_03, bone2Name=dyn_hair_l_fr_02, swing1LimitDegrees=15, swing2LimitDegrees=30, twistLimitDegrees=2}
 {
  bone1Name=dyn_hair_l_mid_01
  bone2Name=head
  swing1LimitDegrees=45
  swing1Motion=Free
  swing2LimitDegrees=45
  swing2Motion=Free
  twistLimitDegrees=45
  twistMotion=Free
 }
 {bone1Name=dyn_hair_l_mid_02, bone2Name=dyn_hair_l_mid_01, swing1LimitDegrees=5, swing2LimitDegrees=30, twistLimitDegrees=2}
 {bone1Name=dyn_hair_l_mid_03, bone2Name=dyn_hair_l_mid_02, swing1LimitDegrees=15, swing2LimitDegrees=30, twistLimitDegrees=2}
 {bone1Name=dyn_hair_l_mid_04, bone2Name=dyn_hair_l_mid_03, swing1LimitDegrees=15, swing2LimitDegrees=30, twistLimitDegrees=2}
 {bone1Name=dyn_hair_l_mid_05, bone2Name=dyn_hair_l_mid_04, swing1LimitDegrees=15, swing2LimitDegrees=30, twistLimitDegrees=2}
 {
  bone1Name=dyn_hair_r_fr_01
  bone2Name=head
  swing1LimitDegrees=45
  swing1Motion=Free
  swing2LimitDegrees=45
  swing2Motion=Free
  twistLimitDegrees=45
  twistMotion=Free
 }
 {bone1Name=dyn_hair_r_fr_02, bone2Name=dyn_hair_r_fr_01, swing1LimitDegrees=15, swing2LimitDegrees=30, twistLimitDegrees=2}
 {bone1Name=dyn_hair_r_fr_03, bone2Name=dyn_hair_r_fr_02, swing1LimitDegrees=15, swing2LimitDegrees=30, twistLimitDegrees=2}
 {bone1Name=dyn_hair_r_fr_04, bone2Name=dyn_hair_r_fr_03, swing1LimitDegrees=15, swing2LimitDegrees=30, twistLimitDegrees=2}
 {
  bone1Name=dyn_hair_r_mid_01
  bone2Name=head
  swing1LimitDegrees=45
  swing1Motion=Free
  swing2LimitDegrees=45
  swing2Motion=Free
  twistLimitDegrees=45
  twistMotion=Free
 }
 {bone1Name=dyn_hair_r_mid_02, bone2Name=dyn_hair_r_mid_01, swing1LimitDegrees=10, swing2LimitDegrees=30, twistLimitDegrees=2}
 {bone1Name=dyn_hair_r_mid_03, bone2Name=dyn_hair_r_mid_02, swing1LimitDegrees=15, swing2LimitDegrees=30, twistLimitDegrees=2}
 {bone1Name=dyn_hair_r_mid_04, bone2Name=dyn_hair_r_mid_03, swing1LimitDegrees=15, swing2LimitDegrees=30, twistLimitDegrees=2}
 {bone1Name=dyn_hair_r_mid_05, bone2Name=dyn_hair_r_mid_04, swing1LimitDegrees=15, swing2LimitDegrees=30, twistLimitDegrees=2}
 {
  bone1Name=dyn_hair_back_03_mstr_l
  bone2Name=head
  swing1LimitDegrees=45
  swing1Motion=Free
  swing2LimitDegrees=45
  swing2Motion=Free
  twistLimitDegrees=45
  twistMotion=Free
 }
 {bone1Name=dyn_hair_back_03_l, bone2Name=dyn_hair_back_03_mstr_l, swing1LimitDegrees=60, swing2LimitDegrees=45, twistLimitDegrees=15}
 {bone1Name=dyn_hair_back_04_l, bone2Name=dyn_hair_back_03_l, swing1LimitDegrees=50, swing2LimitDegrees=45, twistLimitDegrees=5}
 {bone1Name=dyn_hair_back_05_l, bone2Name=dyn_hair_back_04_l, swing1LimitDegrees=60, swing2LimitDegrees=45, twistLimitDegrees=5}
 {
  bone1Name=dyn_hair_back_03_mstr_r
  bone2Name=head
  swing1LimitDegrees=45
  swing1Motion=Free
  swing2LimitDegrees=45
  swing2Motion=Free
  twistLimitDegrees=45
  twistMotion=Free
 }
 {bone1Name=dyn_hair_back_03_r, bone2Name=dyn_hair_back_03_mstr_r, swing1LimitDegrees=60, swing2LimitDegrees=45, twistLimitDegrees=15}
 {bone1Name=dyn_hair_back_04_r, bone2Name=dyn_hair_back_03_r, swing1LimitDegrees=50, swing2LimitDegrees=45, twistLimitDegrees=5}
 {bone1Name=dyn_hair_back_05_r, bone2Name=dyn_hair_back_04_r, swing1LimitDegrees=60, swing2LimitDegrees=45, twistLimitDegrees=5}
 {
  bone1Name=dyn_hair_03_mstr_mid
  bone2Name=head
  swing1LimitDegrees=45
  swing1Motion=Free
  swing2LimitDegrees=45
  swing2Motion=Free
  twistLimitDegrees=45
  twistMotion=Free
 }
 {bone1Name=dyn_hair_03_mid, bone2Name=dyn_hair_03_mstr_mid, swing1LimitDegrees=60, swing2LimitDegrees=45, twistLimitDegrees=15}
 {bone1Name=dyn_hair_04_mid, bone2Name=dyn_hair_03_mid, swing1LimitDegrees=45, swing2LimitDegrees=45, twistLimitDegrees=15}
 {bone1Name=dyn_hair_05_mid, bone2Name=dyn_hair_04_mid, swing1LimitDegrees=45, swing2LimitDegrees=45, twistLimitDegrees=15}
 {
  bone1Name=dyn_hair_back_03_mstr_l
  bone2Name=spine_05
  swing1LimitDegrees=45
  swing1Motion=Free
  swing2LimitDegrees=45
  swing2Motion=Free
  twistLimitDegrees=45
  twistMotion=Free
 }
 {
  bone1Name=dyn_hair_back_04_l
  bone2Name=dyn_hair_04_mid
  swing1LimitDegrees=45
  swing1Motion=Free
  swing2LimitDegrees=45
  swing2Motion=Free
  twistLimitDegrees=45
  twistMotion=Free
 }
 {
  bone1Name=dyn_hair_back_03_mstr_r
  bone2Name=spine_05
  swing1LimitDegrees=45
  swing1Motion=Free
  swing2LimitDegrees=45
  swing2Motion=Free
  twistLimitDegrees=45
  twistMotion=Free
 }
 {
  bone1Name=dyn_hair_back_04_r
  bone2Name=dyn_hair_04_mid
  swing1LimitDegrees=45
  swing1Motion=Free
  swing2LimitDegrees=45
  swing2Motion=Free
  twistLimitDegrees=45
  twistMotion=Free
 }
 {
  bone1Name=dyn_hair_03_mstr_mid
  bone2Name=spine_05
  swing1LimitDegrees=45
  swing1Motion=Free
  swing2LimitDegrees=45
  swing2Motion=Free
  twistLimitDegrees=45
  twistMotion=Free
 }
]
